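; PCG subgraph that copies primary points, splines and 2D polygons onto secondary-input points, preserving each data type.
; It separates inputs by the Input:0 tag, merges secondary points and spline/polygon control points into targets, and stamps each with a temporary targetindex.
; Primary data becomes points, is copied to every target, then partitioned by targetindex.
; Polygons and open or closed splines are rebuilt per partition before the temporary attribute is deleted.
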(graph CopyToPoints_CoreProcess
  (bind DeleteAttributes_0 (Delete%20Attributes :SelectedAttributes "targetindex")) ; Remove temp index attirbute
  ;[0 0 0 0.5 12] Primary/Secondary Inputs
  (bind FilterDataByTag_0 (Filter%20Data%20By%20Tag :SelectedTags "Input:0"))
  ;[0 0 0 0.5] Filter inputs by type (points, splines, polygons) and convert to points before copy
  (bind FilterDataByType_6 (Filter%20Data%20By%20Type :TargetType (struct :Ids [(struct :Struct /Script/PCG.PCGDataTypeInfoPoint)])))
  (bind FilterDataByType_7 (Filter%20Data%20By%20Type :TargetType (struct :Ids [(struct :Struct /Script/PCG.PCGDataTypeInfoSpline)])))
  (bind FilterDataByType_8 (Filter%20Data%20By%20Type :TargetType (struct :Ids [(struct :Struct /Script/PCG.PCGDataTypeInfoPolygon2D)])))
  (bind GetSplineControlPoints_4 (Get%20Spline%20Control%20Points))
  (bind Gather_0 (Gather))
  (bind MergePoints_1 (Merge%20Points :bMergeMetadata false :MergeTagsMode None)) ; Gather all points to copy to
  (bind CopyAttributes_0 (Copy%20Attributes :InputSource $Index :OutputTarget targetindex)) ; Write a temp index for each point to copy to
  (bind FilterDataByType_1 (Filter%20Data%20By%20Type :TargetType (struct :Ids [(struct :Struct /Script/PCG.PCGDataTypeInfoPoint)])))
  (bind MergePoints_0 (Merge%20Points :MergeTagsMode MergeTagsFromFirstPin)) ; Gather all points from primary input as single data
  (bind FilterDataByType_0 (Filter%20Data%20By%20Type :TargetType (struct :Ids [(struct :Struct /Script/PCG.PCGDataTypeInfoSpline)])))
  (bind FilterDataByType_2 (Filter%20Data%20By%20Type :TargetType (struct :Ids [(struct :Struct /Script/PCG.PCGDataTypeInfoPolygon2D)])))
  (bind FilterDataByAttribute_0 (Filter%20Data%20By%20Attribute :FilterMode FilterByValue :FilterOperator Equal :TargetAttribute @Data.$isClosed :Threshold (struct :AttributeTypes (struct :Type Boolean :BoolValue true))))
  (bind GetSplineControlPoints_1 (Get%20Spline%20Control%20Points))
  (bind GetSplineControlPoints_2 (Get%20Spline%20Control%20Points))
  (bind ToPoint_0 (To%20Point))
  ;[0 0 0 0.5] Copy all primary input data and partition them back for each secondary input
  (bind CopyPoints_0 (Copy%20Points :TagInheritance Source))
  (bind CopyPoints_1 (Copy%20Points :TagInheritance Source))
  (bind CopyPoints_2 (Copy%20Points :TagInheritance Source))
  (bind CopyPoints_3 (Copy%20Points :TagInheritance Source))
  (bind AttributePartition_0 (Attribute%20Partition :PartitionAttributeSelectors [targetindex]))
  (bind AttributePartition_1 (Attribute%20Partition :PartitionAttributeSelectors [targetindex]))
  (bind AttributePartition_2 (Attribute%20Partition :PartitionAttributeSelectors [targetindex]))
  (bind AttributePartition_3 (Attribute%20Partition :PartitionAttributeSelectors [targetindex]))
  ;[0 0 0 0.5] Recreate data based on their input types
  (bind CreatePolygon2D_0 (Create%20Polygon%202D))
  (bind CreateSpline_0 (Create%20Spline :bApplyCustomTangents true :bUseInterpTypeAttribute true))
  (bind CreateSpline_1 (Create%20Spline :bApplyCustomTangents true :bClosedLoop true :bUseInterpTypeAttribute true))
  (connect CopyPoints_1 Out AttributePartition_0 In)
  (connect CopyPoints_2 Out AttributePartition_1 In)
  (connect CopyPoints_3 Out AttributePartition_2 In)
  (connect CopyPoints_0 Out AttributePartition_3 In)
  (connect MergePoints_1 Out CopyAttributes_0 Source)
  (connect MergePoints_1 Out CopyAttributes_0 Target)
  (connect MergePoints_0 Out CopyPoints_0 Source)
  (connect CopyAttributes_0 Out CopyPoints_0 Target)
  (connect ToPoint_0 Out CopyPoints_1 Source)
  (connect CopyAttributes_0 Out CopyPoints_1 Target)
  (connect GetSplineControlPoints_2 Out CopyPoints_2 Source)
  (connect CopyAttributes_0 Out CopyPoints_2 Target)
  (connect GetSplineControlPoints_1 Out CopyPoints_3 Source)
  (connect CopyAttributes_0 Out CopyPoints_3 Target)
  (connect AttributePartition_0 Out CreatePolygon2D_0 In)
  (connect AttributePartition_1 Out CreateSpline_0 In)
  (connect AttributePartition_2 Out CreateSpline_1 In)
  (connect AttributePartition_3 Out DeleteAttributes_0 In)
  (connect CreatePolygon2D_0 Out DeleteAttributes_0 In)
  (connect CreateSpline_0 Out DeleteAttributes_0 In)
  (connect CreateSpline_1 Out DeleteAttributes_0 In)
  (connect FilterDataByType_0 InsideFilter FilterDataByAttribute_0 In)
  (connect DefaultInputNode In FilterDataByTag_0 In)
  (connect FilterDataByTag_0 InsideFilter FilterDataByType_0 In)
  (connect FilterDataByTag_0 InsideFilter FilterDataByType_1 In)
  (connect FilterDataByTag_0 InsideFilter FilterDataByType_2 In)
  (connect FilterDataByTag_0 OutsideFilter FilterDataByType_6 In)
  (connect FilterDataByTag_0 OutsideFilter FilterDataByType_7 In)
  (connect FilterDataByTag_0 OutsideFilter FilterDataByType_8 In)
  (connect FilterDataByType_6 InsideFilter Gather_0 In)
  (connect GetSplineControlPoints_4 Out Gather_0 In2)
  (connect FilterDataByAttribute_0 InsideFilter GetSplineControlPoints_1 In)
  (connect FilterDataByAttribute_0 OutsideFilter GetSplineControlPoints_2 In)
  (connect FilterDataByType_7 InsideFilter GetSplineControlPoints_4 In)
  (connect FilterDataByType_8 InsideFilter GetSplineControlPoints_4 In)
  (connect FilterDataByType_1 InsideFilter MergePoints_0 In)
  (connect Gather_0 Out MergePoints_1 In)
  (connect FilterDataByType_2 InsideFilter ToPoint_0 In)
  (output Out DeleteAttributes_0))
